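# Additive unlit mesh-particle material for an explosion sphere in smoke and dust effects, producing a Fresnel glow tinted by particle color.
# MF_Fresnel_FX, with parameterized brightness and exponent, multiplies Particle Color RGB, then feeds a Depth Fade that softens intersections with scene geometry.
# A static switch optionally multiplies in a near-camera distance fade.
# A Quality Switch drops both fades on Low, leaving only the tinted Fresnel.
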Material /Figment_S08/Effects/Fort_Effects/Gameplay/SmokeDustExplosions/Materials/M_ExplosionSphere_01
graph {
 properties {bUsedWithMeshParticles=true, blendMode=BLEND_Additive, pixelDepthOffsetMode=PDOM_Legacy, shadingModel=MSM_Unlit, usedShadingModels=Unlit}
 nodes [
  n1 = Multiply(
   A=ParticleColor().RGB
   B=MaterialFunctionCall(
    Brightness=ScalarParameter() {defaultValue=1, parameterName=Fresnel Brightness}
    Exponent=ScalarParameter() {defaultValue=10, parameterName=Fresnel Exponent}
   ) {materialFunction=&MF_Fresnel_FX}.Fresnel
  )
  n2 = ScalarParameter() {defaultValue=500, parameterName=CameraFadeDistance}
  n3 = Multiply(
   A=DepthFade(Opacity=@n1, FadeDistance=ScalarParameter() {defaultValue=500, parameterName=Depth Fade Distance})
   B=StaticSwitchParameter(
    True=MaterialFunctionCall(Distance=@n2) {materialFunction=&Near_Camera_Fade}.Result
    False=Constant() {r=1}
   ) {parameterName=UseCameraFade}
  )
  QualitySwitch(Default=@n3, Low=@n1, High=@n3, Medium=@n3)
 ]
}
